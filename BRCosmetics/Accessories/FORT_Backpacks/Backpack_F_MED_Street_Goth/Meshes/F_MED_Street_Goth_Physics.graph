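# Physics asset for the Street Goth backpack cosmetic, giving its bear-shaped body ragdoll-style bodies for secondary motion.
# A small box at the backpack root anchors upper and lower bear body capsules, each with moderate angular limits.
# Arms, ears, tongue and zipper hang off the upper body, and legs off the lower body.
# Limbs and zipper swing mostly on one axis, while ears and tongue swing equally on both with tight twist.
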PhysicsAsset /BRCosmetics/Accessories/FORT_Backpacks/Backpack_F_MED_Street_Goth/Meshes/F_MED_Street_Goth_Physics
bodies {
 backpack_root [{shapeType=Box, extentX=1.25, extentY=1.25, extentZ=1.25}]
 bear_lowebody [{shapeType=Capsule, length=0, radius=2.5}]
 bear_upperbody [{shapeType=Capsule, length=0, radius=2.5}]
 dyn_arm_l [{shapeType=Capsule, length=0, radius=2.5, center=<9.4463 0.8264 1.8914>}]
 dyn_arm_r [{shapeType=Capsule, length=0, radius=2.5, center=<9.8179 0.859 -1.7088>}]
 dyn_ear_l [{shapeType=Capsule, length=2.5, radius=2.5}]
 dyn_ear_r [{shapeType=Capsule, length=2.5, radius=2.5}]
 dyn_leg_l [{shapeType=Capsule, length=0, radius=2.5, center=<7.3035 0 5.9478>}]
 dyn_leg_r [{shapeType=Capsule, length=0, radius=2.5, center=<7.4034 0 -5.3478>}]
 dyn_tongue [{shapeType=Capsule, length=0, radius=1, center=<2.8 -0.0001 0.6147>}]
 dyn_zipper [{shapeType=Capsule, length=0, radius=1, center=<1 0 0>}]
}
constraints [
 {bone1Name=dyn_leg_r, bone2Name=bear_lowebody, swing1LimitDegrees=40, swing2LimitDegrees=5, twistLimitDegrees=5}
 {bone1Name=bear_upperbody, bone2Name=backpack_root, swing1LimitDegrees=15, swing2LimitDegrees=15, twistLimitDegrees=15}
 {bone1Name=dyn_arm_l, bone2Name=bear_upperbody, swing1LimitDegrees=40, swing2LimitDegrees=1, twistLimitDegrees=1}
 {bone1Name=dyn_zipper, bone2Name=bear_upperbody, swing1LimitDegrees=60, swing2LimitDegrees=5, twistLimitDegrees=5}
 {bone1Name=dyn_ear_l, bone2Name=bear_upperbody, swing1LimitDegrees=25, swing2LimitDegrees=25, twistLimitDegrees=5}
 {bone1Name=dyn_tongue, bone2Name=bear_upperbody, swing1LimitDegrees=11, swing2LimitDegrees=11, twistLimitDegrees=5}
 {bone1Name=dyn_arm_r, bone2Name=bear_upperbody, swing1LimitDegrees=40, swing2LimitDegrees=1, twistLimitDegrees=1}
 {bone1Name=dyn_ear_r, bone2Name=bear_upperbody, swing1LimitDegrees=25, swing2LimitDegrees=25, twistLimitDegrees=5}
 {bone1Name=bear_lowebody, bone2Name=backpack_root, swing1LimitDegrees=15, swing2LimitDegrees=15, twistLimitDegrees=15}
 {bone1Name=dyn_leg_l, bone2Name=bear_lowebody, swing1LimitDegrees=40, swing2LimitDegrees=5, twistLimitDegrees=5}
]
